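; Parent gameplay ability for Athena active powers, providing shared activation, cast and cleanup logic.
; It applies a power-active effect on init, removes it and optional hover on end, and can refuse activation while swimming.
; The cast montage routes notifies to overridable events, stops on commit failure, and a retriggerable failsafe delay can end the ability.
; Helpers toggle server-side movement blocking, SetByCaller-driven hover gravity with velocity damping, and pawn collision.
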(event Ability|EventOnEndAbility
  (params (bWasCancelled :type Boolean))
  ;[1 1 1] On End Ability
  (|CallParentFunction bWasCancelled)
  ;[1 1 1] Remove Active Power Active Tag
  (Ability|RemoveGameplayEffectFromOwnerWithHandle (Variables|PowerActivation|GetGEH_PowerActive))
  ;[1 1 1] End Hover On End Ability
  (if (and (Variables|PowerHover|GetIsHovering) (Variables|PowerHover|GetRemoveHoverOnAbilityEnd))
    (CallFunction|SetHover false)))

(event Custom|StartFailsafeDelay
  ;[1 1 1] Start Failsafe Delay - Call this when you want to start the failsafe. Call it again if it needs to be extended.
  (Utilities|FlowControl|RetriggerableDelay (Variables|PowerFailsafe|GetFailsafeDelay))
  (Ability|EndAbility))

(event Custom|CastMontageNotifyStart
  (params (NotifyTag :type Gameplay%20Tag%20Structure)))

(event Custom|CastMontageNotifyEnd
  (params (NotifyTag :type Gameplay%20Tag%20Structure)))

(event Custom|CastMontageCancelledOrInterrupted)

(event Custom|CastMontageNotifyEndAbility)

(event Custom|CastMontageBlendOut)

(event EventPowerCommitFailed
  ;[1 1 1] Event Power Commit Failed - If we fail the commit. End the montage.
  (Animation|StopAnimMontage (Variables|Default|GetPlayerPawn) (Variables|PowerAnimationCast|GetCastMontage))
  (|CallParentFunction))

(event Custom|PlayCastAnimation
  ;[1 1 1] On Ability Activated - Play the montage and start the failsafe
  (bind (_asynctaskproxy _notifytag) (Ability|Tasks|PlayMontageAndWaitNotify "Cast Animation" (CallFunction|GetCastMontage) (Variables|PowerAnimationCast|GetCastMontagePlayRate) (Variables|PowerAnimationCast|GetCastMontageStartSection) (Variables|PowerAnimationCast|GetCastMontageStopWhenAbilityEnds) (Variables|PowerAnimationCast|GetCastMontageRootMotionScale))
    (:OnCompleted
      ;[1 1 1] Should the ability end when the montage ends.
      (if (Variables|PowerAnimationCast|GetEndAbilityOnMontageComplete)
        (Ability|EndAbility)))
    (:OnBlendOut
      (CallFunction|CastMontageBlendOut))
    (:OnInterrupted
      _)
    (:OnCancelled
      _)
    (:OnNotifyBegin
      (CallFunction|CastMontageNotifyStart _notifytag))
    (:OnNotifyEnd
      (CallFunction|CastMontageNotifyEnd _notifytag)))
  (CallFunction|CastMontageCancelledorInterrupted))

(fn GetCastMontage
  (return (Variables|PowerAnimationCast|GetCastMontage)))

(fn InitAbility
  ;[1 1 1] Grant the Active Power Active Tag.
  (bind _returnvalue (Ability|ApplyGameplayEffectToOwner (Variables|PowerActivation|GetGE_PowerActive)))
  (Variables|PowerActivation|SetGEH_PowerActive _returnvalue)
  (|CallParentFunction))

(fn CanActivateAbility
  (params (ActorInfo :type Gameplay%20Ability%20Actor%20Info%20Structure) (Handle :type Gameplay%20Ability%20Spec%20Handle%20Structure))
  (bind (_owneractor _avataractor) (Utilities|Struct|BreakGameplayAbilityActorInfo ActorInfo))
  ;[1 1 1] Check if the ability can be activated while swimming if it cannot and the player is swimming send fail message for swimming
  (bind _asfort_player_pawn (Utilities|Casting|CastToFortPlayerPawn _avataractor)
    (:then
      (return (Variables|PowerActivation|GetTag_ActivationFailedReasonSwimming) (or (Variables|PowerActivation|GetAllowActivationWhileSwimming) (not (GameplayTags|TagContainer|CompareTagAssetInterfacetoOtherTags (Variables|PowerActivation|GetTag_SurfaceSwimming) _asfort_player_pawn)))))
    (:CastFailed
      (return 0 false))))

(fn SetBlockMovement
  (params (PlayerPawn :type Fort%20Player%20Pawn%20Object%20Reference) (Block :type Boolean) (HasDuration :type Boolean) (Duration :type Float%20%28double-precision%29 :default 1.0))
  ;[1 1 1] Set player to movement to Blocking or Blocking.
  (Utilities|IsServer
    (:true
      (if Block
        (if HasDuration
          (bind _returnvalue (Ability|GameplayEffect|SetDuration (Ability|MakeOutgoingGameplayEffectSpec (Variables|PowerPreventMove|GetGE_PreventMoveDuration)) Duration))
          (bind _returnvalue_2 (Ability|ApplyGameplayEffectSpecToOwner _returnvalue))
          (Variables|PowerPreventMove|SetHandle_PreventMove _returnvalue_2)
          (else
            (bind _returnvalue_3 (Ability|ApplyGameplayEffectToOwner (Variables|PowerPreventMove|GetGE_PreventMove)))
            (Variables|PowerPreventMove|SetHandle_PreventMove _returnvalue_3)))
        (else
          (Ability|RemoveGameplayEffectFromOwnerWithHandle (Variables|PowerPreventMove|GetHandle_PreventMove))))
      (Input|SetRotatePawntoCamera (Fortnite|GetFortPlayerControllerfromActor PlayerPawn) (not Block)))))

(fn SetHover
  (params (On :type Boolean))
  (bind _playerpawn (Variables|Default|GetPlayerPawn))
  (bind _charactermovement (Variables|/Script/Engine.Character|GetCharacterMovement _playerpawn))
  (bind _velocity (Variables|/Script/Engine.MovementComponent|GetVelocity _charactermovement))
  (bind _hovervelocitymaintained (Variables|PowerHover|GetHoverVelocityMaintained))
  ;[1 1 1] Using a GE with custom magnitude, set a gravity override on the player and save out the handle. Do checks for velocity overrides and movement blocking assigning that as needed. Set the Velocity to clamp the Z to a positive value. Additionally, undoes all the setting with the On==False option.
  (Utilities|IsValid (Variables|Default|GetPlayerPawn)
    (:Is%20Valid
      (Variables|PowerHover|SetIsHovering On)
      (if On
        (Pawn|Character|ResetFallingHeight (Variables|Default|GetPlayerPawn))
        (bind _returnvalue (Ability|GameplayEffect|AssignTagSetbyCallerMagnitude (Ability|MakeOutgoingGameplayEffectSpec (Variables|PowerHover|GetGE_HoverGravity) 0.0) "(TagName=\"SetByCaller.Magnitude\")" (Variables|PowerHover|GetHoverGravityMultiplier)))
        (bind _returnvalue_2 (Ability|GameplayEffect|AssignTagSetbyCallerMagnitude _returnvalue "(TagName=\"SetByCaller.MoveSpeed\")" (Variables|PowerHover|GetHoverAirControlAmount)))
        (bind _returnvalue_3 (Ability|ApplyGameplayEffectSpecToOwner _returnvalue_2))
        (Variables|PowerHover|SetGEH_Hover _returnvalue_3)
        (if (>= _hovervelocitymaintained 1.0)
          (Variables|/Script/Engine.MovementComponent|SetVelocity (Utilities|PositiveClampZ _velocity) _charactermovement)
          (else
            (Variables|/Script/Engine.MovementComponent|SetVelocity (Utilities|PositiveClampZ (* _velocity _hovervelocitymaintained)) _charactermovement)
            (Character|StopJumping _playerpawn)))
        (else
          (Pawn|Character|ResetFallingHeight (Variables|Default|GetPlayerPawn))
          (Ability|RemoveGameplayEffectFromOwnerWithHandle (Variables|PowerHover|GetGEH_Hover)))))
    (:Is%20Not%20Valid
      (Development|PrintString "Power: Set Hover called but pawn is not valid."))))

(fn SetPawnCollision
  (params (Character :type Character%20Object%20Reference) (CollisionOn :type Boolean))
  (bind _capsulecomponent (Variables|/Script/Engine.Character|GetCapsuleComponent Character))
  ;[1 1 1] Check if the player is valid and if we are turning on or off.
  (Utilities|IsValid Character
    (:Is%20Valid
      (if CollisionOn
        ;[1 1 1] Set Pawn collision back to blocking if the saved collision is not none
        (Collision|SetCollisionResponsetoChannel _capsulecomponent "ECC_Pawn" "ECR_Block")
        (Collision|SetCollisionResponsetoChannel _capsulecomponent "ECC_GameTraceChannel4" "ECR_Block")
        (else
          ;[1 1 1] Set Pawn collision to overlap and save out collision
          (Collision|SetCollisionResponsetoChannel _capsulecomponent "ECC_Pawn" "ECR_Overlap")
          (Collision|SetCollisionResponsetoChannel _capsulecomponent "ECC_GameTraceChannel4" "ECR_Overlap"))))
    (:Is%20Not%20Valid
      (Development|PrintString "Power: Set Pawn Collision Character Not Valid"))))
