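; Parent gameplay ability for Athena active powers with standard activation, firing OnAbilityActivated either on press or on input release.
; Activation sets up cancel tag and water-entry bindings, effect removal, camera, timed hover and cast montage.
; Release mode runs a held-input cue, optional movement block and max-charge timer, then a DoOnce, cooldown-checked release derives a curve-based charge multiplier from hold time.
; EndAbility undoes movement, hover, pawn collision and charging state.
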(event Custom|InitActivateOnRelease
  ;[1 1 1] Init Activate On Release -  Set up looping Cue. Check if movement should be blocked, and set up the hit max charge check if it is applicable
  (CallFunction|AddGCLoop (Variables|PowerGameplayCues|GetGC_InputHeldLoop) (Variables|PowerGameplayCues|GetInputHeldLoopGCUsesParams) (CallFunction|InputHeldLoopGCParams) true)
  (if (Variables|PowerActivation|GetBlockMovementDuringHold)
    (CallFunction|SetBlockMovement (Variables|Default|GetPlayerPawn) true))
  (if (Variables|PowerCharge|GetCharges)
    (Variables|PowerCharge|SetCharging true)
    (bind _returnvalue (Utilities|Time|SetTimerbyEvent (EventDispatchers|CreateEvent :!event Custom|ReachedMaxCharge) (ScalableFloat|GetValueatLevel (Variables|PowerCharge|GetRow_MaxChargeTime)))) ; Set Timer For when we hit max charge
    (Variables|PowerCharge|SetTimer_ReachedMaxCharge _returnvalue)))

(event Custom|ReachedMaxCharge
  ;[1 1 1] Reached Max Charge - Execute Max Charge GC
  (CallFunction|PlayBurstGC (Variables|PowerGameplayCues|GetGC_MaxCharge) (Variables|PowerGameplayCues|GetMaxChargeGCUsesParams) (CallFunction|MaxChargeGCParams)))

(event Custom|Released
  (params (HoldTime :type Float%20%28double-precision%29))
  ;[1 1 1] Check if the ability activates on press or release. Init the ability and check activation and release times if applicable. Call the On Ability Activated event to play the meat of most abilities and start failsafe timer.
  (if (Variables|PowerActivation|GetActivatesOnRelease)
    (Utilities|FlowControl|DoOnce false
      (:Completed
        ;;[1 1 1] Check if we should fire if the ability is on cooldown.
        (bind _outfortplayerpawn (CallFunction|GetAvatarasValidFortPlayerPawn)
          (:ValidFortPlayerPawn
            (bind _returnvalue (Ability|Cooldowns|IsCooldownTagsActive _outfortplayerpawn (Ability|Cooldowns|GetCooldownTags)))
            (if (or (not _returnvalue) (and _returnvalue (Variables|PowerCooldown|GetReleaseFiresIfOnCooldown)))
              (CallFunction|CalculateReleaseTime HoldTime)
              (if (Variables|PowerEffectRemoval|GetRemoveEffectsOnRelease)
                (CallFunction|RemoveSpecifiedEffects))
              (if (Variables|PowerHover|GetStopHoverOnRelease)
                (CallFunction|SetHover false))
              (Utilities|Time|ClearandInvalidateTimerbyHandle (Variables|PowerCharge|GetTimer_ReachedMaxCharge))
              (Ability|RemoveGameplayCueFromOwner (Variables|PowerGameplayCues|GetGC_InputHeldLoop))
              (if (Variables|PowerActivation|GetBlockMovementDuringHold)
                (CallFunction|SetBlockMovement (Variables|Default|GetPlayerPawn) true) ; Unblock Player movement
              )
              (CallFunction|ClearReleaseCameraMode)
              (CallFunction|OnAbilityActivated)
              (Development|PrintString "Power: Ability Activated")
              (CallFunction|StartFailsafeDelay)
              ;[1 1 1] Check if the player should have collision with other pawns during the ability
              (if (Variables|PowerCollision|GetOverlapPawnCollisionDuringAbility)
                (CallFunction|SetPawnCollision (Variables|Default|GetPlayerPawn)))))))) ; Makes it so release can not fire multiple times Per activation
  ))

(event Custom|ResetActivationReleaseDoOnce
  ;[1 1 1] Check if the ability activates on press or release. Init the ability and check activation and release times if applicable. Call the On Ability Activated event to play the meat of most abilities and start failsafe timer.
  (Utilities|FlowControl|DoOnce false
    (:Completed
      ;;[1 1 1] Check if we should fire if the ability is on cooldown.
      (bind _outfortplayerpawn (CallFunction|GetAvatarasValidFortPlayerPawn)
        (:ValidFortPlayerPawn
          (bind _returnvalue (Ability|Cooldowns|IsCooldownTagsActive _outfortplayerpawn (Ability|Cooldowns|GetCooldownTags)))
          (if (or (not _returnvalue) (and _returnvalue (Variables|PowerCooldown|GetReleaseFiresIfOnCooldown)))
            (CallFunction|CalculateReleaseTime (stale Custom|Released HoldTime))
            (if (Variables|PowerEffectRemoval|GetRemoveEffectsOnRelease)
              (CallFunction|RemoveSpecifiedEffects))
            (if (Variables|PowerHover|GetStopHoverOnRelease)
              (CallFunction|SetHover false))
            (Utilities|Time|ClearandInvalidateTimerbyHandle (Variables|PowerCharge|GetTimer_ReachedMaxCharge))
            (Ability|RemoveGameplayCueFromOwner (Variables|PowerGameplayCues|GetGC_InputHeldLoop))
            (if (Variables|PowerActivation|GetBlockMovementDuringHold)
              (CallFunction|SetBlockMovement (Variables|Default|GetPlayerPawn) true) ; Unblock Player movement
            )
            (CallFunction|ClearReleaseCameraMode)
            (CallFunction|OnAbilityActivated)
            (Development|PrintString "Power: Ability Activated")
            (CallFunction|StartFailsafeDelay)
            ;[1 1 1] Check if the player should have collision with other pawns during the ability
            (if (Variables|PowerCollision|GetOverlapPawnCollisionDuringAbility)
              (CallFunction|SetPawnCollision (Variables|Default|GetPlayerPawn)))))))) ; Makes it so release can not fire multiple times Per activation
)

(event Ability|EventActivateAbility
  ;[1 1 1] Check if the ability activates on press or release. Init the ability and check activation and release times if applicable. Call the On Ability Activated event to play the meat of most abilities and start failsafe timer.
  (bind (_asynctaskproxy _timeheld) (Ability|Tasks|WaitInputRelease (Variables|PowerActivation|GetTestAlreadyReleased))
    (:then
      (CallFunction|SetupCancelBinds)
      (CallFunction|InitAbility)
      (if (Variables|PowerEffectRemoval|GetRemoveEffectsOnPress)
        (CallFunction|RemoveSpecifiedEffects))
      (CallFunction|SetWaterBinding)
      (CallFunction|SetValidCamera (Variables|PowerCamera|GetActivateCameraMode))
      (if (Variables|PowerHover|GetHoverOnActivate)
        (CallFunction|SetHover true)
        (Utilities|FlowControl|Sequence
          (:then_0
            _)
          (:then_1
            ;[1 1 1] Hover Time Limit
            (if (> (Variables|PowerHover|GetHoverTimeLimit) 0.0)
              (bind _returnvalue (Utilities|Time|SetTimerbyEvent (delegate Custom|TurnOffHover) (Variables|PowerHover|GetHoverTimeLimit)))
              (Variables|PowerHover|SetHoverTimer _returnvalue)))))
      (if (Variables|PowerAnimationCast|GetPlayCastOnActivate)
        (CallFunction|PlayCastAnimation))
      (if (Variables|PowerActivation|GetActivatesOnRelease)
        (CallFunction|InitActivateonRelease)
        (else
          (CallFunction|OnAbilityActivated)
          (Development|PrintString "Power: Ability Activated")
          (CallFunction|StartFailsafeDelay)
          ;[1 1 1] Check if the player should have collision with other pawns during the ability
          (if (Variables|PowerCollision|GetOverlapPawnCollisionDuringAbility)
            (CallFunction|SetPawnCollision (Variables|Default|GetPlayerPawn))))))
    (:OnRelease
      (CallFunction|Released _timeheld))))

(event Custom|OnAbilityActivated
  ;[1 1 1] On Ability Activated - If set to, jump to the cast montage section in the ability.
  (Variables|PowerCharge|SetCharging false)
  (if (Variables|PowerAnimationCast|GetOnActivatedJumpToCastSection)
    (Ability|Animation|MontageJumptoSection (Variables|PowerAnimationCast|GetCastSectionName))
    (Networking|ForceNetUpdate (Variables|Default|GetPlayerPawn))))

(event Custom|SetupCancelBinds
  ;[1 1 1] Setup Cancel Binds - Wait for things that should end the ability, like DBNO.
  (Ability|Tasks|WaitGameplayTagAdd (Variables|PowerEndAbility|GetTag_CancelPower)
    (:then
      (if (Variables|PowerEndAbility|GetEndOnKnockback)
        (Ability|Tasks|WaitGameplayTagAdd (Variables|PowerEndAbility|GetTag_Knockback)
          (:then
            (if (Variables|PowerEndAbility|GetEndOnDBNO) :!label _label
              (Ability|Tasks|WaitGameplayTagAdd (Variables|PowerEndAbility|GetTag_DBNO)
                (:Added
                  _))))
          (:Added
            _))
        (bind _outfortplayerpawn (CallFunction|GetAvatarasValidFortPlayerPawn :!label _label_2)
          (:ValidFortPlayerPawn
            ;;[1 1 1] End the montage if it is the cast montage
            (if (Utilities|Equal(Object) (Animation|GetCurrentMontage _outfortplayerpawn :!class /Script/Engine.Character) (Variables|PowerAnimationCast|GetCastMontage))
              (Ability|Animation|MontageStop -1.0)))
          (:AvatarCastFailed
            _))
        (Networking|ForceNetUpdate (Ability|GetAvatarActorfromActorInfo))
        (Ability|EndAbility))
      (goto _label))
    (:Added
      _))
  (goto _label_2))

(event Custom|SetWaterBinding
  ;[1 1 1] Handle Water binding - Bind to the player entering water checking if the ability should end if the player enters water
  (EventDispatchers|BindEventtoOnPawnEnterWater (Variables|Default|GetPlayerPawn) (delegate Custom|PlayerEnterWater)))

(event Custom|TurnOffHover
  ;[1 1 1] Hover Time Limit
  (CallFunction|SetHover false))

(event Custom|PlayerEnterWater
  (params (FortPlayerPawn :type Fort%20Player%20Pawn%20Object%20Reference))
  ;[1 1 1] Handle Water binding - Bind to the player entering water checking if the ability should end if the player enters water
  (if (Variables|PowerWater|GetEndOnWaterEnter)
    (Animation|StopAnimMontage (Variables|Default|GetPlayerPawn) (Variables|PowerAnimationCast|GetCastMontage))
    (EventDispatchers|UnbindEventfromOnPawnEnterWater (Variables|Default|GetPlayerPawn) (delegate Custom|PlayerEnterWater))
    (Development|PrintString "Power: Ended from entering water")
    (Ability|EndAbility)))

(event Ability|EventOnEndAbility
  (params (bWasCancelled :type Boolean))
  ;[1 1 1] On End Ability
  (|CallParentFunction bWasCancelled)
  (Utilities|FlowControl|Sequence
    (:then_0
      ;[1 1 1] Unblock Player Movement if it was being blocked
      (if (Variables|PowerActivation|GetBlockMovementDuringHold)
        (CallFunction|SetBlockMovement (Variables|Default|GetPlayerPawn)) ; Unblock Player movement
      ))
    (:then_1
      ;[1 1 1] Set Hover off if it was turned on during the ability
      (if (Variables|PowerHover|GetRemoveHoverOnAbilityEnd)
        (CallFunction|SetHover false)))
    (:then_2
      ;[1 1 1] Turn pawn collision back on if it was turned off.
      (if (Variables|PowerCollision|GetOverlapPawnCollisionDuringAbility)
        (CallFunction|SetPawnCollision (Variables|Default|GetPlayerPawn) true)))
    (:then_3
      ;[1 1 1] Turn Charges off on end if the ability charges.
      (if (Variables|PowerCharge|GetCharges)
        (Variables|PowerCharge|SetCharging false)))
    (:then_4
      ;[1 1 1] Unbind OnPawnEnterWater
      (EventDispatchers|UnbindEventfromOnPawnEnterWater (Variables|Default|GetPlayerPawn) (EventDispatchers|CreateEvent :!event Custom|PlayerEnterWater)))) ; Undo things that may have been turned on.
)

(fn CalculateCharge
  (params (ChargeTime :type Float%20%28double-precision%29))
  ;[1 1 1] Find the charge percent based on activation and release time
  (bind _output_get (Variables|PowerCharge|SetChargeTime ChargeTime))
  (bind _output_get_2 (Variables|PowerCharge|SetChargePercent (Math|Float|Clamp(Float) (/ _output_get (ScalableFloat|GetValueatLevel (Variables|PowerCharge|GetRow_MaxChargeTime))) (Variables|PowerCharge|GetChargePercentMin) (Variables|PowerCharge|GetChargePercentMax))))
  ;[1 1 1] Calculate charge multiplier from the Curve
  (Variables|PowerCharge|SetChargeMultiplier (Math|Curves|GetFloatValue (Variables|PowerCharge|GetCurve_ChargeMultiplier) _output_get_2)))

(fn CalculateReleaseTime
  (params (ReleaseTime :type Float%20%28double-precision%29))
  ;[1 1 1] If the ability charges calc the charge time
  (if (Variables|PowerCharge|GetCharges)
    (CallFunction|CalculateCharge ReleaseTime)))

(fn ClearReleaseCameraMode
  ;[1 1 1] Clear the camera mode on release if not.
  (if (Variables|PowerCamera|GetClearsCameraModeOnRelease)
    (Camera|PopCameraMode)))

(fn InitAbility
  ;[1 1 1] When the ability is initiated reset the activation release do once.
  (CallFunction|ResetActivationReleaseDoOnce)
  (|CallParentFunction))

(fn InputHeldLoopGCParams
  (return 0))

(fn MaxChargeGCParams
  (return 0))

(fn RemoveSpecifiedEffects
  ;[1 1 1] Effects we want removed when ability is activated.  FeatherFall abilities are removed when activating another ability by default.
  (for _array_element (Variables|PowerEffectRemoval|GetGrantedTagsToRemove)
    (Ability|RemoveGameplayEffectFromOwnerWithGrantedTags (GameplayTags|MakeGameplayTagContainerfromTag _array_element)))
  (for _array_element_2 (Variables|PowerEffectRemoval|GetAssetTagsToRemove)
    (Ability|RemoveGameplayEffectFromOwnerWithAssetTags (GameplayTags|MakeGameplayTagContainerfromTag _array_element_2))))

(fn SetValidCamera
  (params (InputPin :type Fort%20Camera%20Mode%20Class%20Reference))
  ;[1 1 1] Set Camera if valid
  (if (Utilities|IsValidClass InputPin)
    (Camera|PushCameraMode InputPin)))
