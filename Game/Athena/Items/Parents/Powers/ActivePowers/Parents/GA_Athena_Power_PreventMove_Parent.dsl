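; Parent Gameplay Ability for Athena active powers that, by its name, prevent movement.
; On activation it only prints a "Movement Blocked!!" debug string.
; A graph note states a passive must be activated to block movement, but no blocking, ending or commit logic appears here.
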(event Ability|EventActivateAbility
  ;[1 1 1] Passive needs to be activated to block movement
  (Development|PrintString "Movement Blocked!!"))
